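# Vertex-painted ground material for Volcano environments, blending two world-aligned texture sets, defaulting to Pirate Cove stonewall and cobblestone.
# Red and green tint each layer, alpha drives their height blend, and world-space noise varies brightness.
# Blue adds height-blended wetness and puddles with panning normals, Fresnel darkening and optional oil sheen or underwater lighting, before move-tool and event functions.
# Low quality and non-deferred paths keep only layer one, without puddles.
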Material /Game/Athena/Environments/Volcano/Materials/M_GroundPaint
graph {
 properties {bUseMaterialAttributes=true, pixelDepthOffsetMode=PDOM_Legacy, usedShadingModels=Default Lit}
 nodes [
  n1 = VectorParameter() {desc="Green ", group=2, parameterName="2 Tint (G Ch VP)", sortPriority=5, defaultValue=<rgba 1 1 1 0>}
  n2 = TextureObjectParameter() {group=2, parameterName=2D, sortPriority=8, texture=/Game/Environments/Sets/Pirate_Cove/Pirate_Village/Textures/T_Cove_CobbleStone_Floor_D}
  n3 = ScalarParameter() {defaultValue=327.5804, group=2, parameterName=2 UVScale, sortPriority=9}
  n4 = MaterialFunctionCall(TextureObject=@n2, TextureSize=@n3, Export Float 4=StaticBool() {value=true}) {materialFunction=&WorldAlignedTexture}
  n5 = ComponentMask(None=@n4.Z Texture) {b=true, g=true, r=true}
  n6 = TextureObjectParameter() {
   group=2
   parameterName=2S
   samplerType=SAMPLERTYPE_Masks
   sortPriority=8
   texture=/Game/Environments/Sets/Pirate_Cove/Pirate_Village/Textures/T_Cove_CobbleStone_Floor_S
  }
  n7 = MaterialFunctionCall(Float3=MaterialFunctionCall(TextureObject=@n6, TextureSize=@n3) {materialFunction=&WorldAlignedTexture}.Z Texture) {materialFunction=&BreakOutFloat3Components}
  n8 = TextureObjectParameter() {
   group=2
   parameterName=2N
   samplerType=SAMPLERTYPE_Normal
   sortPriority=8
   texture=/Game/Environments/Sets/Pirate_Cove/Pirate_Village/Textures/T_Cove_CobbleStone_Floor_N
  }
  n9 = ScalarParameter() {defaultValue=1, group=2, parameterName=2N Val, sortPriority=7}
  n10 = AppendVector(A=@n9, B=@n9)
  n11 = Constant() {r=1}
  n12 = SetMaterialAttributes!(
   MaterialAttributes
   Base Color=LinearInterpolate(A=@n5, B=Multiply(A=@n1.RGB, B=@n5), Alpha=VertexColor().G)
   Metallic=@n7.G
   Specular=@n7.G
   Roughness=Multiply(A=@n7.B, B=ScalarParameter() {defaultValue=1, group=2, parameterName=2RScale, sortPriority=6})
   Normal=Multiply(
    A=MaterialFunctionCall(TextureObject=@n8, TextureSize=@n3) {materialFunction=&WorldAlignedNormal}.Z Texture
    B=AppendVector(A=@n10, B=@n11)
   )
  ) {attributeSetTypes [
    #1
    #2
    #3
    #4
    #5
   ]}
  n13 = VectorParameter() {group=1, parameterName="1 Tint (R Ch VP)", defaultValue=<rgba 1 1 1 0>}
  n14 = TextureObjectParameter() {group=1, parameterName=D, sortPriority=3, texture=/Game/Environments/Sets/Pirate_Cove/Pirate_Village/Textures/T_Cove_Stonewall_D}
  n15 = ScalarParameter() {defaultValue=327.5804, group=1, parameterName=1 UVScale, sortPriority=4}
  n16 = MaterialFunctionCall(TextureObject=@n14, TextureSize=@n15, Export Float 4=StaticBool() {value=true}) {materialFunction=&WorldAlignedTexture}
  n17 = ComponentMask(None=@n16.Z Texture) {b=true, g=true, r=true}
  n18 = TextureObjectParameter() {
   group=1
   parameterName=S
   samplerType=SAMPLERTYPE_Masks
   sortPriority=3
   texture=/Game/Environments/Sets/Pirate_Cove/Pirate_Village/Textures/T_Cove_Stonewall_S
  }
  n19 = MaterialFunctionCall(Float3=MaterialFunctionCall(TextureObject=@n18, TextureSize=@n15) {materialFunction=&WorldAlignedTexture}.Z Texture) {materialFunction=&BreakOutFloat3Components}
  n20 = TextureObjectParameter() {
   group=1
   parameterName=N
   samplerType=SAMPLERTYPE_Normal
   sortPriority=3
   texture=/Game/Environments/Sets/Pirate_Cove/Pirate_Village/Textures/T_Cove_Stonewall_N
  }
  n21 = ScalarParameter() {defaultValue=1, group=1, parameterName=N Val, sortPriority=2}
  n22 = AppendVector(A=@n21, B=@n21)
  n23 = Constant() {r=1}
  n24 = SetMaterialAttributes(
   Base Color=LinearInterpolate(A=@n17, B=Multiply(A=@n13.RGB, B=@n17), Alpha=VertexColor().R)
   Metallic=@n19.G
   Specular=@n19.G
   Roughness=Multiply(A=@n19.B, B=ScalarParameter() {defaultValue=1, group=1, parameterName=RScale, sortPriority=1})
   Normal=Multiply(
    A=MaterialFunctionCall(TextureObject=@n20, TextureSize=@n15) {materialFunction=&WorldAlignedNormal}.Z Texture
    B=AppendVector(A=@n22, B=@n23)
   )
  ) {attributeSetTypes [
    #1
    #2
    #3
    #4
    #5
   ]}
  n25 = ComponentMask(None=@n16.Z Texture) {a=true}
  n26 = MaterialFunctionCall!(
   Diffuse 1
   Diffuse 2
   Transistion Phase=VertexColor().A
   Height Texture 1=@n25
   Height Texture 2=ComponentMask(None=@n4.Z Texture) {a=true}
   Contrast=ScalarParameter() {defaultValue=3, group=Blend, parameterName=Blend_Contrast}
  ) {materialFunction=/Engine/Functions/Engine_MaterialFunctions02/Texturing/HeightLerpWithTwoHeightMaps}
  n27 = StaticSwitchParameter(
   True=@n24
   False=MaterialFunctionCall!(Base Material=@n12, Top Material=@n24, Alpha=@n26.Alpha) {materialFunction=/Engine/Functions/MaterialLayerFunctions/MatLayerBlend_Standard}.Blended Material
  ) {parameterName=Use Single Layer}
  n28 = GetMaterialAttributes(None=ShadingPathSwitch(Default=@n24, Deferred=QualitySwitch(Default=@n27, Low=@n24))) {attributeGetTypes [#1]}
  n29 = ComponentMask(None=WorldPosition().XYZ) {g=true, r=true}
  n30 = TextureSample(UVs=Divide(A=@n29, B=ScalarParameter() {defaultValue=1024, parameterName=Global Noise})) {samplerType=SAMPLERTYPE_LinearColor, texture=/Game/Global/Textures/Default/Noise/T_Noise01_Inverse}
  n31 = SetMaterialAttributes!(
   MaterialAttributes=@n28.MaterialAttributes
   Base Color=Multiply(A=@n28.Base Color, B=LinearInterpolate(Alpha=@n30.R) {constA=0.9, constB=1.1})
  ) {attributeSetTypes [#1]}
  n32 = GetMaterialAttributes(None=@n31) {attributeGetTypes [#4, #3, #1]}
  n33 = Constant() {r=-0.5}
  n34 = Constant() {r=1}
  n35 = ComponentMask(None=@n32.Roughness) {r=true}
  n36 = Divide(A=@n29, B=ScalarParameter() {defaultValue=279.4971, group=Puddles, parameterName=Puddles Scale})
  n37 = TextureSample(UVs=@n36) {samplerType=SAMPLERTYPE_LinearColor, texture=&T_Noise01_Inverse}
  n38 = MaterialFunctionCall(
   Base Material=@n31
   Top Material=SetMaterialAttributes!(
    MaterialAttributes=@n32.MaterialAttributes
    Roughness=LinearInterpolate(A=@n33, B=@n34, Alpha=@n35)
    Specular=Multiply(A=@n32.Specular) {constB=0.25}
    Base Color=Multiply(A=@n32.Base Color) {constB=0.5}
   ) {attributeSetTypes [#4, #3, #1]}
   Alpha=MaterialFunctionCall(Transition Phase=VertexColor().B, Height Texture=Multiply(B=@n37.R), Contrast=Constant() {r=0.5}) {materialFunction=&HeightLerp}.Alpha
  ) {materialFunction=&MatLayerBlend_Standard}
  n39 = GetMaterialAttributes(None=@n38.Blended Material) {attributeGetTypes [#1]}
  n40 = Constant3Vector() {constant=<rgba 0 0 1 0>}
  n41 = ComponentMask(None=WorldPosition().XYZ) {g=true, r=true}
  n42 = Constant() {r=64}
  n43 = TextureSample(UVs=Panner(Coordinate=Divide(A=@n41, B=@n42)) {speedX=0.1, speedY=0.1}) {samplerSource=SSM_Wrap_WorldGroupSettings, samplerType=SAMPLERTYPE_Normal, texture=/Game/Global/Textures/Default/Noise/T_SwirlNoise_N}
  n44 = LinearInterpolate(
   A=@n40
   B=MaterialFunctionCall(Normal=@n43.RGB, Flatness=Constant() {r=0.995}) {materialFunction=&FlattenNormal}.Result
   Alpha=@n37.R
  )
  n45 = Fresnel() {baseReflectFraction=0.2, exponent=3}
  n46 = Multiply(A=Multiply(A=@n39.Base Color, B=ScalarParameter() {defaultValue=0.75, group=Puddles, parameterName=Puddles Mod}), B=OneMinus(None=@n45))
  n47 = LinearInterpolate(A=Constant() {r=0.5}, Alpha=@n45)
  n48 = ReflectionVectorWS(CustomWorldNormal=GetMaterialAttributes(None=@n31) {attributeGetTypes [#5]}.Normal)
  n49 = AtmosphericLightVector()
  n50 = Saturate(None=Power(Base=DotProduct(A=@n48, B=@n49), Exp=Constant() {r=2}))
  n51 = StaticSwitchParameter(True=Multiply(A=@n46, B=LinearInterpolate(A=@n47, Alpha=@n50)), False=@n46) {group=Extra, parameterName=UnderWaterLighting true}
  n52 = Divide(A=@n36, B=Constant() {r=-4})
  n53 = Fresnel(Normal=Transform(None=TextureSample(UVs=@n52) {samplerType=SAMPLERTYPE_Normal, texture=&T_SwirlNoise_N}.RGB))
  n54 = Multiply(
   A=Multiply(A=@n46, B=TextureSample(UVs=@n53) {texture=/Game/Characters/Fortress_CharacterShader/Textures/RainbowEffect}.RGB)
   B=Multiply(A=@n45, B=ScalarParameter() {defaultValue=0.3, group=Extra, parameterName=Oilness})
  ) {constB=5}
  n55 = SetMaterialAttributes!(
   MaterialAttributes=@n39.MaterialAttributes
   Normal=ShadingPathSwitch(Default=@n40, Deferred=QualitySwitch(Default=@n44, Low=@n40))
   Roughness=Constant() {r=0.1}
   Specular=LinearInterpolate(A=Constant() {r=0.02}, B=Constant() {r=0.2}, Alpha=@n45)
   Base Color=StaticSwitchParameter(True=Add(A=@n51, B=@n54), False=@n51) {group=Extra, parameterName=Oilness true}
  ) {attributeSetTypes [
    #5
    #4
    #3
    #1
   ]}
  n56 = VertexColor()
  n57 = Constant() {r=-0.2}
  n58 = OneMinus(None=StaticSwitchParameter(True=@n25, False=@n26.Resulting Height Map) {parameterName=Use Single Layer})
  n59 = ShadingPathSwitch(Default=@n31, Deferred=QualitySwitch(Default=@n38.Blended Material, Low=@n31))
  n60 = MaterialFunctionCall(
   Base Material=@n59
   Top Material=@n55
   Alpha=MaterialFunctionCall(
    Transition Phase=Add(A=@n56.B, B=@n57)
    Height Texture=Add(A=@n58, B=@n37.R)
    Contrast=ScalarParameter() {defaultValue=1, group=Puddles, parameterName=PuddlesContrast}
   ) {materialFunction=&HeightLerp}.Alpha
  ) {materialFunction=&MatLayerBlend_Standard}
  n61 = MaterialFunctionCall(MA In=ShadingPathSwitch(Default=@n59, Deferred=QualitySwitch(Default=@n60.Blended Material, Low=@n59))) {materialFunction=&MF_CreativeMoveTool}
  MaterialFunctionCall!(In=@n61.MA Out) {materialFunction=/Game/Athena/WinterArt/Materials/MF_Events}
  n62 = Constant3Vector() {constant=<rgba 0.9375 0.0391 0.0391 0>}
  MaterialFunctionCall(
   Base Material=@n61.MA Out
   Top Material=MakeMaterialAttributes!(
    BaseColor=@n62
    Metallic
    Specular
    Roughness
    Anisotropy
    EmissiveColor=@n62
    Opacity
    OpacityMask
    Normal
    Tangent
    WorldPositionOffset
    SubsurfaceColor
    ClearCoat
    ClearCoatRoughness
    AmbientOcclusion
    Refraction
    CustomizedUVs_0
    CustomizedUVs_1
    CustomizedUVs_2
    CustomizedUVs_3
    CustomizedUVs_4
    CustomizedUVs_5
    CustomizedUVs_6
    CustomizedUVs_7
    PixelDepthOffset
    ShadingModel
    Displacement
   )
   Alpha=Constant() {r=1}
  ) {materialFunction=&MatLayerBlend_Standard}
  Multiply()
 ]
}
