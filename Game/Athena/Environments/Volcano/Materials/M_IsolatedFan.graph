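# Material for a spinning metal fan in the Volcano silo environment, wrapped in the shared Fortnite base function.
# World position offset rotates vertices around the object pivot on a parameterized local axis over time, with normals corrected to match.
# One mask channel drives metallic, specular and roughness.
# A camera-distance term named WPODisable is built to blend toward static attributes, but its floored math evaluates to 1 at every distance.
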Material /Game/Athena/Environments/Volcano/Materials/M_IsolatedFan
graph {
 properties {
  bUseMaterialAttributes=true
  bUsedWithNanite=true
  physMaterial=/Game/FortressPhysicalMaterials/Metal
  pixelDepthOffsetMode=PDOM_Legacy
  refractionMethod=RM_IndexOfRefraction
  usedShadingModels=Default Lit
 }
 nodes [
  n1 = TextureSampleParameter2D() {parameterName=Diffuse, texture=/Game/Athena/Environments/Volcano/Textures/Silo/T_volcano_silo_fan_D}
  n2 = TextureSampleParameter2D() {parameterName=SpecularMasks, samplerType=SAMPLERTYPE_Masks, texture=/Game/Athena/Environments/Volcano/Textures/Silo/T_volcano_silo_fan_S}
  n3 = TextureSampleParameter2D() {parameterName=Normals, samplerType=SAMPLERTYPE_Normal, texture=/Game/Athena/Environments/Volcano/Textures/Silo/T_volcano_silo_fan_N}
  n4 = Normalize(VectorInput=Transform(None=VectorParameter() {parameterName=Vector Rotation Direction, defaultValue=<rgba 0 1 0 1>}.RGB) {transformSourceType=TRANSFORMSOURCE_Local})
  n5 = Multiply(A=Time(), B=ScalarParameter() {defaultValue=1, parameterName=Rotation Speed})
  n6 = MaterialFunctionCall(Rotation Axis=@n4, Rotation Angle=@n5, World Space Vertex Normals=Transform(None=@n3.RGB)) {materialFunction=&FixRotateAboutAxisNormals}
  n7 = RotateAboutAxis(
   NormalizedRotationAxis=@n4
   RotationAngle=@n5
   PivotPoint=MaterialFunctionCall() {materialFunction=&ObjectPivotPoint}.Object Pivot Location
   Position=WorldPosition().XYZ
  )
  MaterialFunctionCall!(
   Material attributes=BlendMaterialAttributes(
    A=MakeMaterialAttributes!(
     BaseColor=@n1.RGB
     Metallic=@n2.G
     Specular=@n2.G
     Roughness=@n2.G
     Anisotropy
     EmissiveColor
     Opacity
     OpacityMask
     Normal=@n3.RGB
     Tangent
     WorldPositionOffset
     SubsurfaceColor
     ClearCoat
     ClearCoatRoughness
     AmbientOcclusion
     Refraction
     CustomizedUVs_0
     CustomizedUVs_1
     CustomizedUVs_2
     CustomizedUVs_3
     CustomizedUVs_4
     CustomizedUVs_5
     CustomizedUVs_6
     CustomizedUVs_7
     PixelDepthOffset
     ShadingModel
     Displacement
    )
    B=MakeMaterialAttributes(BaseColor=@n1.RGB, Metallic=@n2.G, Specular=@n2.G, Roughness=@n2.G, Normal=@n6.New Tangent Space Vertex Normal, WorldPositionOffset=@n7)
    Alpha=NamedRerouteUsage()
   )
   IsGameplayAsset=StaticBool() {value=true}
  ) {materialFunction=/Game/Packages/Fortress_SharedMaterials/Base_Material_Functions/Fortnite_BaseFunction}
  n8 = Distance(
   A=MaterialFunctionCall() {materialFunction=&ObjectPivotPoint}.Object Pivot Location
   B=CameraPositionWS()
  )
  n9 = Constant() {desc=offset, r=-5000}
  n10 = Floor(None=Divide(A=Add(A=@n8, B=@n9), B=Constant() {desc=Gradient, r=-35000}) {constB=35000}) {desc=clamp value}
  NamedRerouteDeclaration(None=OneMinus(None=Saturate(None=@n10))) {name=WPODisable, nodeColor=<rgba 0 0.3906 1 1>}
  WorldPosition()
 ]
}
