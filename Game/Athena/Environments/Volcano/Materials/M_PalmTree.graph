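# Palm tree foliage material in the Athena Volcano environment, using two-sided foliage shading with Nanite usage enabled.
# Pre-skinned height plus transformed normal-map Z forms a gradient that blends a ground-contact tint into the diffuse and adjusts roughness.
# TwoSidedSign varies specular and roughness and limits a faint diffuse-based emissive to back faces, while a mask channel scales subsurface color.
# The assembled attributes pass through MF_4Level_TreeAnimations for tree animation.
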Material /Game/Athena/Environments/Volcano/Materials/M_PalmTree
graph {
 properties {
  bUseMaterialAttributes=true
  bUsedWithNanite=true
  pixelDepthOffsetMode=PDOM_Legacy
  shadingModel=MSM_TwoSidedFoliage
  twoSided=true
  usedShadingModels=Two Sided Foliage
 }
 nodes [
  n1 = VectorParameter() {parameterName=GroundContactColor, defaultValue=<rgba 0.0527 0.099 0.0103 0>}
  n2 = TextureSampleParameter2D() {parameterName=Diffuse, texture=/Game/Environments/Props/Trees/Textures/T_PalmTree_A_D}
  n3 = Divide(A=VertexInterpolator(VS=PreSkinnedPosition()).PS, B=ScalarParameter() {defaultValue=500, parameterName=GroundGradientHeight})
  n4 = TextureSampleParameter2D() {parameterName=Normals, samplerType=SAMPLERTYPE_Normal, texture=/Game/Environments/Props/Trees/Textures/T_PalmTree_A_N}
  n5 = Transform(None=@n4.RGB)
  n6 = Saturate(None=Add(A=ComponentMask(None=@n3) {b=true}, B=ComponentMask(None=@n5) {b=true}))
  n7 = TextureSampleParameter2D() {parameterName=SpecularMasks, samplerType=SAMPLERTYPE_Masks, texture=/Game/Environments/Props/Trees/Textures/T_PalmTree_A_S}
  n8 = TwoSidedSign()
  n9 = Constant() {r=0.5}
  n10 = LinearInterpolate(A=Constant() {r=0.6}, B=@n7.B, Alpha=@n8)
  n11 = LinearInterpolate(A=@n2.RGB, Alpha=@n8) {constB=0}
  n12 = ScalarParameter() {defaultValue=0.05, parameterName=EmissiveBottom}
  n13 = VectorParameter() {parameterName=SSS, defaultValue=<rgba 0.1525 0.4167 0.0471 0>}
  MaterialFunctionCall!(In=MakeMaterialAttributes!(
    BaseColor=LinearInterpolate(A=@n1.RGB, B=@n2.RGB, Alpha=@n6)
    Metallic
    Specular=LinearInterpolate(B=@n7.R, Alpha=@n8) {constA=0.1}
    Roughness=LinearInterpolate(A=@n9, B=@n10, Alpha=@n6)
    Anisotropy
    EmissiveColor=Multiply(A=@n11, B=@n12)
    Opacity
    OpacityMask=@n2.A
    Normal=@n4.RGB
    Tangent
    WorldPositionOffset
    SubsurfaceColor=Multiply(A=@n7.G, B=@n13.RGB)
    ClearCoat
    ClearCoatRoughness
    AmbientOcclusion
    Refraction
    CustomizedUVs_0
    CustomizedUVs_1
    CustomizedUVs_2
    CustomizedUVs_3
    CustomizedUVs_4
    CustomizedUVs_5
    CustomizedUVs_6
    CustomizedUVs_7
    PixelDepthOffset
    ShadingModel
    Displacement
   )) {materialFunction=/Game/Packages/Fortnite_UniqueMaterials_1/MaterialFunctions/Trees/MF_4Level_TreeAnimations}
 ]
}
